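# Translucent, per-pixel-lit glass master material for the LastVoiceSteel female character body, with screen-space reflections.
# Character PreFX feeds optional Slurp, with Thin Film wired only into its enabled branch, both masked by SkinFX_Mask channels.
# A shared tint then overrides base color and emissive (optionally time-of-day boosted), while Fresnel drives opacity alongside parameterized specular and roughness.
# The result passes through cloth-sim WPO into character PostFX.
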Material /BRCosmetics/Characters/Player/Female/Medium/Bodies/F_MED_LastVoiceSteel/Materials/M_LastVoiceSteel_Glass_MASTER
graph {
 properties {
  bScreenSpaceReflections=true
  bUseMaterialAttributes=true
  bUsedWithMeshDeformer=true
  bUsedWithMorphTargets=true
  bUsedWithSkeletalMesh=true
  blendMode=BLEND_Translucent
  pixelDepthOffsetMode=PDOM_Legacy
  translucencyLightingMode=TLM_SurfacePerPixelLighting
  usedShadingModels=Default Lit
 }
 nodes [
  n1 = MaterialFunctionCall() {materialFunction=&BaseFN_CharacterFunction_PreFX}
  n2 = TextureSample(Tex=TextureObjectParameter() {group=Skin FX, parameterName=SkinFX_Mask})
  n3 = StaticComponentMaskParameter(None=@n2.RGBA) {defaultR=true, group=Skin FX, parameterName=Slurp Channel}
  n4 = StaticComponentMaskParameter(None=@n2.RGBA) {defaultG=true, group=Skin FX, parameterName=Slurp Bubbles Channel}
  n5 = StaticComponentMaskParameter(None=@n2.RGBA) {defaultB=true, group=Skin FX, parameterName=Slurp Normal Distortion Channel}
  n6 = StaticSwitchParameter(
   True=MaterialFunctionCall!(
    Alpha=StaticComponentMaskParameter(None=@n2.RGBA) {defaultR=true, group=Skin FX, parameterName=ThinFilm_Channel}
    Material Attributes=@n1.Result
   ) {materialFunction=/Game/Weapons/WeaponSkins/MaterialFunction/ThinFilmFunction}.Result
   False=@n1.Result
  ) {group=Skin FX, parameterName=Use Thin Film}
  n7 = GetMaterialAttributes(None=StaticSwitchParameter(
    True=MaterialFunctionCall!(Alpha=@n3, AO=@n1.AO, BubblesMask=@n4, SlurpNormalDistortionMask=@n5, Pre Skinned Position=@n1.AO, Material=@n6) {materialFunction=/Game/Weapons/WeaponSkins/MaterialFunction/MF_Slurp}.Result
    False=@n1.Result
   ) {group=Skin FX, parameterName=UseSlurp}) {attributeGetTypes [#1, #2]}
  n8 = NamedRerouteUsage()
  n9 = SetMaterialAttributes!(
   MaterialAttributes=@n7.MaterialAttributes
   Opacity=NamedRerouteUsage()
   Specular=ScalarParameter() {defaultValue=1, group=LastVoiceSteel, parameterName=GlassSpecular}
   Roughness=ScalarParameter() {group=LastVoiceSteel, parameterName=GlassRoughness}
   Emissive Color=StaticSwitchParameter(True=MaterialFunctionCall!(In=@n8) {materialFunction=/Game/Weapons/WeaponSkins/MaterialFunction/BoostEmissivebyTOD}.Result, False=@n8) {group=Skin FX, parameterName=Use TOD Emissive Boost}
  ) {attributeSetTypes [
    #3
    #4
    #5
    #2
   ]}
  n10 = GetMaterialAttributes(None=@n9) {attributeGetTypes [#2]}
  n11 = Multiply(A=@n10.Emissive Color, B=ScalarParameter() {defaultValue=0.5, group=LastVoiceSteel, parameterName=LastVoiceSteel_EmissiveBoost})
  n12 = VectorParameter() {group=LastVoiceSteel, parameterName=EmissiveColor, defaultValue=<rgba 0.2986 0.5509 0.776 1>}
  n13 = Multiply(A=@n12.RGB, B=ScalarParameter() {defaultValue=0.25, group=LastVoiceSteel, parameterName=LastVoiceSteel_BaseColorBoost})
  MaterialFunctionCall(
   AO=@n1.AO
   Distance Blend=@n1.AO
   Fresnel=@n1.AO
   Gmap=@n1.AO
   Material=MaterialFunctionCall!(
    Material=SetMaterialAttributes!(MaterialAttributes=@n10.MaterialAttributes, Emissive Color=@n11, Base Color=@n13) {attributeSetTypes [#2, #1]}
    SkinFX_Mask=NamedRerouteUsage()
   ) {materialFunction=/Game/Global/MaterialFunctions/MF_ClothSimWPO_v2}.Result
  ) {materialFunction=&BaseFN_CharacterFunction_PostFX}
  NamedRerouteDeclaration(None=@n7.MaterialAttributes) {name=MA_IN, nodeColor=<rgba 0 0.4141 1 1>}
  NamedRerouteDeclaration(None=Multiply(A=@n12.RGB, B=ScalarParameter() {defaultValue=0.6, group=LastVoiceSteel, parameterName=EmissiveStrength})) {name=Emissive_OUT, nodeColor=<rgba 0.3594 0 1 1>}
  n14 = Power(
   Base=Fresnel(
    ExponentIn=ScalarParameter() {defaultValue=1.5, group=LastVoiceSteel, parameterName=FresnelExp}
    BaseReflectFractionIn=ScalarParameter() {group=LastVoiceSteel, parameterName=FresnelBase}
   )
   Exp=ScalarParameter() {defaultValue=1, group=LastVoiceSteel, parameterName=FresnelPower}
  )
  n15 = ScalarParameter() {defaultValue=3, group=LastVoiceSteel, parameterName=FresnelMulti}
  NamedRerouteDeclaration(None=Saturate(None=Multiply(A=@n14, B=@n15) {constB=11})) {name=Opacity_OUT, nodeColor=<rgba 1 0.8672 0 1>}
  NamedRerouteDeclaration(None=@n2.RGBA) {name=SkinFX_Mask_Sampled, nodeColor=<rgba 0.7475 0 1 1>}
 ]
}
