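# Physics asset for the jacket flaps on the LastVoiceSteel female medium player cosmetic body.
# Each side has a mirrored capsule chain running from a Master bone through two flap segments.
# Constraints allow only a narrow swing on one axis, tighter at the second segment, with zero limits on the other swing and twist.
# Larger upper-arm capsules sit outside the chains, unconstrained, as collision shapes.
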PhysicsAsset /BRCosmetics/Characters/Player/Female/Medium/Bodies/F_MED_LastVoiceSteel/Meshes/F_MED_LastVoiceSteel_JacketFlaps_Physics
bodies {
 dyn_jacket_l_A_01 [{shapeType=Capsule, length=8, radius=3, shapeName=dyn_jacket_l_A_01_capsule, center=<2.2222 0 0>, rotation=<rot 0 0 90>}]
 dyn_jacket_l_A_02 [{shapeType=Capsule, length=8, radius=3, shapeName=dyn_jacket_l_A_02_capsule, center=<2.2222 0 0>, rotation=<rot 0 0 90>}]
 dyn_jacket_l_A_Master [{shapeType=Capsule, length=2, radius=3, shapeName=dyn_jacket_l_A_Master_capsule, center=<2.2222 0 0>}]
 dyn_jacket_r_A_01 [{shapeType=Capsule, length=8, radius=3, shapeName=dyn_jacket_l_A_01_capsule, center=<2.3739 0 0>, rotation=<rot 0 0 -90>}]
 dyn_jacket_r_A_02 [{shapeType=Capsule, length=8, radius=3, shapeName=dyn_jacket_l_A_02_capsule, center=<2.3739 0 0>, rotation=<rot 0 0 -90>}]
 dyn_jacket_r_A_Master [{shapeType=Capsule, length=2, radius=3, shapeName=dyn_jacket_l_A_Master_capsule, center=<-2.2222 0 0>, rotation=<rot 0 0 -180>}]
 upperarm_l [{shapeType=Capsule, length=14.6016, radius=7, shapeName=upperarm_l_capsule, center=<17.7865 0 0>, rotation=<rot 90 0 0>}]
 upperarm_r [{shapeType=Capsule, length=14.6016, radius=7, shapeName=upperarm_l_capsule, center=<-17.7865 0 0>, rotation=<rot 90 0 180>}]
}
constraints [
 {bone1Name=dyn_jacket_l_A_01, bone2Name=dyn_jacket_l_A_Master, swing1LimitDegrees=0, swing2LimitDegrees=15, twistLimitDegrees=0}
 {bone1Name=dyn_jacket_l_A_02, bone2Name=dyn_jacket_l_A_01, swing1LimitDegrees=0, swing2LimitDegrees=10, twistLimitDegrees=0}
 {bone1Name=dyn_jacket_r_A_02, bone2Name=dyn_jacket_r_A_01, swing1LimitDegrees=0, swing2LimitDegrees=10, twistLimitDegrees=0}
 {bone1Name=dyn_jacket_r_A_01, bone2Name=dyn_jacket_r_A_Master, swing1LimitDegrees=0, swing2LimitDegrees=15, twistLimitDegrees=0}
]
