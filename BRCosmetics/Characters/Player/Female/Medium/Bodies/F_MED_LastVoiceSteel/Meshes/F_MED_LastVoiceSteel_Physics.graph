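# Physics asset for the female medium LastVoiceSteel body cosmetic, set up for dynamic jacket bones.
# Mirrored B and C jacket chains hang from spine_02 master bones, with tight angular limits below the masters and near-spherical capsules on C segments.
# Capsules on spine, arms and thighs supply body shapes.
# Spine and arm joints swing freely except the right shoulder, which is limited; thighs have no constraints.
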PhysicsAsset /BRCosmetics/Characters/Player/Female/Medium/Bodies/F_MED_LastVoiceSteel/Meshes/F_MED_LastVoiceSteel_Physics
bodies {
 dyn_jacket_l_B_01 [{shapeType=Capsule, length=0.505, radius=0.505, shapeName=dyn_jacket_l_B_01_capsule, center=<1.4304 0 0>, rotation=<rot 90 0 0>}]
 dyn_jacket_l_B_02 [{shapeType=Capsule, length=5.7226, radius=2.6903, shapeName=dyn_jacket_l_B_02_capsule, center=<1.4304 0 0>, rotation=<rot 90 0 0>}]
 dyn_jacket_l_B_Master [{shapeType=Capsule, length=0.505, radius=0.505, shapeName=dyn_jacket_l_B_Master_capsule, center=<1.4304 0 0>, rotation=<rot 90 0 0>}]
 dyn_jacket_l_C_01 [{shapeType=Capsule, length=0, radius=6, shapeName=dyn_jacket_l_C_01_capsule, center=<2.8313 0 0>}]
 dyn_jacket_l_C_02 [{shapeType=Capsule, length=0, radius=6, shapeName=dyn_jacket_l_C_02_capsule, center=<2.4626 0 0>}]
 dyn_jacket_l_C_03 [{shapeType=Capsule, length=0, radius=6, shapeName=dyn_jacket_l_C_03_capsule, center=<3.1731 0 0>}]
 dyn_jacket_l_C_Master [{shapeType=Capsule, length=0.505, radius=0.505, shapeName=dyn_jacket_l_C_Master_capsule, rotation=<rot 0 -90 0>}]
 dyn_jacket_r_B_01 [{shapeType=Capsule, length=0.505, radius=0.505, shapeName=dyn_jacket_r_B_01_capsule, center=<1.4304 0 0>, rotation=<rot 90 0 0>}]
 dyn_jacket_r_B_02 [{shapeType=Capsule, length=5.7226, radius=2.6903, shapeName=dyn_jacket_r_B_02_capsule, center=<1.4304 0 0>, rotation=<rot 90 0 0>}]
 dyn_jacket_r_B_Master [{shapeType=Capsule, length=0.505, radius=0.505, shapeName=dyn_jacket_r_B_Master_capsule, center=<1.4304 0 0>, rotation=<rot 90 0 0>}]
 dyn_jacket_r_C_01 [{shapeType=Capsule, length=0, radius=6, shapeName=dyn_jacket_l_C_01_capsule, center=<1.9776 0 0>, rotation=<rot 0 0 -180>}]
 dyn_jacket_r_C_02 [{shapeType=Capsule, length=0, radius=6, shapeName=dyn_jacket_l_C_02_capsule, center=<2.3064 0 0>, rotation=<rot 0 0 -180>}]
 dyn_jacket_r_C_03 [{shapeType=Capsule, length=0, radius=6, shapeName=dyn_jacket_l_C_03_capsule, center=<2.3881 0 0>, rotation=<rot 0 0 -180>}]
 dyn_jacket_r_C_Master [{shapeType=Capsule, length=0.505, radius=0.505, shapeName=dyn_jacket_r_C_Master_capsule, rotation=<rot 0 -90 0>}]
 lowerarm_l [{shapeType=Capsule, length=4, radius=6.5, shapeName=lowerarm_l_capsule, center=<12.1772 0 0>, rotation=<rot 90 0 0>}]
 lowerarm_r [{shapeType=Capsule, length=4, radius=6.5, shapeName=lowerarm_l_capsule, center=<-12.1772 0 0>, rotation=<rot 90 180 0>}]
 spine_01 [{shapeType=Capsule, length=3, radius=8, shapeName=spine_01_capsule}]
 spine_02 [{shapeType=Capsule, length=1.25, radius=7.5, shapeName=spine_02_capsule}]
 thigh_l [{shapeType=Capsule, length=22.2221, radius=7.5, center=<-20.2917 -1.4684 -0.4515>, rotation=<rot 0 88.0543 -82.7985>}]
 thigh_r [{shapeType=Capsule, length=22.2315, radius=7.5, center=<20.2579 1.4903 0.498>, rotation=<rot 0 88.1867 -83.7625>}]
 upperarm_l [{shapeType=Capsule, length=5.3516, radius=5, shapeName=upperarm_l_capsule, center=<12.7243 0 0>, rotation=<rot 90 0 0>}]
 upperarm_r [{shapeType=Capsule, length=5.3516, radius=5, shapeName=upperarm_l_capsule, center=<-12.7243 0 0>, rotation=<rot 90 0 180>}]
}
constraints [
 {bone1Name=dyn_jacket_l_C_01, bone2Name=dyn_jacket_l_C_Master, swing1LimitDegrees=0, swing2LimitDegrees=15, twistLimitDegrees=0}
 {bone1Name=dyn_jacket_l_C_02, bone2Name=dyn_jacket_l_C_01, swing1LimitDegrees=0, swing2LimitDegrees=15, twistLimitDegrees=0}
 {bone1Name=dyn_jacket_l_C_03, bone2Name=dyn_jacket_l_C_02, swing1LimitDegrees=0, swing2LimitDegrees=15, twistLimitDegrees=0}
 {bone1Name=dyn_jacket_r_C_01, bone2Name=dyn_jacket_r_C_Master, swing1LimitDegrees=0, swing2LimitDegrees=15, twistLimitDegrees=0}
 {bone1Name=dyn_jacket_r_C_02, bone2Name=dyn_jacket_r_C_01, swing1LimitDegrees=0, swing2LimitDegrees=15, twistLimitDegrees=0}
 {bone1Name=dyn_jacket_r_C_03, bone2Name=dyn_jacket_r_C_02, swing1LimitDegrees=0, swing2LimitDegrees=15, twistLimitDegrees=0}
 {
  bone1Name=lowerarm_l
  bone2Name=upperarm_l
  swing1LimitDegrees=45
  swing1Motion=Free
  swing2LimitDegrees=45
  swing2Motion=Free
  twistLimitDegrees=45
  twistMotion=Free
 }
 {
  bone1Name=dyn_jacket_l_C_Master
  bone2Name=spine_02
  swing1LimitDegrees=45
  swing1Motion=Free
  swing2LimitDegrees=45
  swing2Motion=Free
  twistLimitDegrees=45
  twistMotion=Free
 }
 {
  bone1Name=dyn_jacket_r_C_Master
  bone2Name=spine_02
  swing1LimitDegrees=45
  swing1Motion=Free
  swing2LimitDegrees=45
  swing2Motion=Free
  twistLimitDegrees=45
  twistMotion=Free
 }
 {
  bone1Name=spine_02
  bone2Name=spine_01
  swing1LimitDegrees=45
  swing1Motion=Free
  swing2LimitDegrees=45
  swing2Motion=Free
  twistLimitDegrees=45
  twistMotion=Free
 }
 {
  bone1Name=upperarm_l
  bone2Name=spine_02
  swing1LimitDegrees=45
  swing1Motion=Free
  swing2LimitDegrees=45
  swing2Motion=Free
  twistLimitDegrees=45
  twistMotion=Free
 }
 {bone1Name=dyn_jacket_l_B_02, bone2Name=dyn_jacket_l_B_01, swing1LimitDegrees=5, swing2LimitDegrees=15, twistLimitDegrees=0}
 {bone1Name=dyn_jacket_r_B_02, bone2Name=dyn_jacket_r_B_01, swing1LimitDegrees=5, swing2LimitDegrees=15, twistLimitDegrees=0}
 {bone1Name=dyn_jacket_l_B_Master, bone2Name=spine_02, swing1LimitDegrees=5, swing2LimitDegrees=15, twistLimitDegrees=0}
 {bone1Name=dyn_jacket_l_B_01, bone2Name=dyn_jacket_l_B_Master, swing1LimitDegrees=5, swing2LimitDegrees=15, twistLimitDegrees=0}
 {bone1Name=dyn_jacket_r_B_Master, bone2Name=spine_02, swing1LimitDegrees=5, swing2LimitDegrees=15, twistLimitDegrees=0}
 {bone1Name=dyn_jacket_r_B_01, bone2Name=dyn_jacket_r_B_Master, swing1LimitDegrees=5, swing2LimitDegrees=15, twistLimitDegrees=0}
 {
  bone1Name=lowerarm_r
  bone2Name=upperarm_r
  swing1LimitDegrees=45
  swing1Motion=Free
  swing2LimitDegrees=45
  swing2Motion=Free
  twistLimitDegrees=45
  twistMotion=Free
 }
 {bone1Name=upperarm_r, bone2Name=spine_02, swing1LimitDegrees=5, swing2LimitDegrees=15, twistLimitDegrees=0}
]
