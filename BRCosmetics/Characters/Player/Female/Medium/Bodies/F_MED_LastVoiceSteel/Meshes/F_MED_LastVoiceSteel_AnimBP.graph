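# Animation Blueprint for the F_MED_LastVoiceSteel female body cosmetic, adding secondary motion to its jacket and body.
# The parent mesh pose passes through deform and jacket Control Rigs, then two RigidBody simulations using gravity set at initialization.
# A front-end idle curve blends DefaultSlot onto the jacket branches.
# Shoulder-corrective constraints and bone offsets, their alpha calculations and WPO wind logic exist but are disconnected.
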AnimBlueprint /BRCosmetics/Characters/Player/Female/Medium/Bodies/F_MED_LastVoiceSteel/Meshes/F_MED_LastVoiceSteel_AnimBP
target_skeleton=/Game/Characters/Player/Female/Medium/Bodies/F_MED_LastVoiceSteel/Meshes/F_MED_LastVoiceSteel_Skeleton
graphs [
 AnimGraph {nodes [
   n1 = "Constraint - flont_uppelalm_l (PP)"!(
    ComponentPose=Local To Component(LocalPose=Copy Pose From Mesh() {bUseAttachedParent=true})
    Alpha=1
    ConstraintWeights_0=0.5
    ConstraintWeights_1:float=0.5
   ) -> (Pose) {
    alphaBoolBlend {blendOption=Linear}
    boneToModify {boneName=flont_uppelalm_l}
    constraintSetup [{targetBone {boneName=upperarm_twist_01_l}}, {targetBone {boneName=clavicle_l}}]
    constraintWeights [1,1]
   }
   n2 = "Constraint - up_uppelalm_l (PP)"!(ComponentPose=@n1, Alpha=1, ConstraintWeights_0=0.5, ConstraintWeights_1:float=0.5) -> (Pose) {
    alphaBoolBlend {blendOption=Linear}
    boneToModify {boneName=up_uppelalm_l}
    constraintSetup [{targetBone {boneName=upperarm_twist_01_l}}, {targetBone {boneName=clavicle_l}}]
    constraintWeights [1,1]
   }
   n3 = "Constraint - front_upperarm_r (PP)"!(ComponentPose=@n2, Alpha=1, ConstraintWeights_0=0.5, ConstraintWeights_1:float=0.5) -> (Pose) {
    alphaBoolBlend {blendOption=Linear}
    boneToModify {boneName=front_upperarm_r}
    constraintSetup [{targetBone {boneName=upperarm_twist_01_r}}, {targetBone {boneName=clavicle_r}}]
    constraintWeights [1,1]
   }
   n4 = "Constraint - up_upperarm_r (PP)"!(ComponentPose=@n3, Alpha=1, ConstraintWeights_0=0.5, ConstraintWeights_1:float=0.5) -> (Pose) {
    alphaBoolBlend {blendOption=Linear}
    boneToModify {boneName=up_upperarm_r}
    constraintSetup [{targetBone {boneName=upperarm_twist_01_r}}, {targetBone {boneName=clavicle_r}}]
    constraintWeights [1,1]
   }
   n5 = Get Shoulder_rot_f_l() -> (Shoulder_rot_f_l:double)
   n6 = "Transform (Modify) Bone - Bone: flont_uppelalm_l"(ComponentPose=@n4, Alpha=@n5, Translation="-5,5,0", Rotation="0,0,0", Scale="1,1,1") {
    rotationMode=BMM_Additive
    rotationSpace=BCS_BoneSpace
    scaleMode=BMM_Additive
    scaleSpace=BCS_BoneSpace
    translationMode=BMM_Additive
    translationSpace=BCS_BoneSpace
    alphaBoolBlend {blendOption=Linear}
    boneToModify {boneName=flont_uppelalm_l}
    translation=<-5 5>
   }
   n7 = Get Shoulder_rot_f_r() -> (Shoulder_rot_f_r:double)
   n8 = "Transform (Modify) Bone - Bone: front_upperarm_r"(ComponentPose=@n6, Alpha=@n7, Translation="5,-5,0", Rotation="0,0,0", Scale="1,1,1") {
    rotationMode=BMM_Additive
    rotationSpace=BCS_BoneSpace
    scaleMode=BMM_Additive
    scaleSpace=BCS_BoneSpace
    translationMode=BMM_Additive
    translationSpace=BCS_BoneSpace
    alphaBoolBlend {blendOption=Linear}
    boneToModify {boneName=front_upperarm_r}
    translation=<5 -5>
   }
   n9 = "Transform (Modify) Bone - Bone: up_uppelalm_l"(ComponentPose=@n8, Alpha=@n5, Translation="0,2,3", Rotation="0,0,0", Scale="1,1,1") {
    rotationMode=BMM_Additive
    rotationSpace=BCS_BoneSpace
    scaleMode=BMM_Additive
    scaleSpace=BCS_BoneSpace
    translationMode=BMM_Additive
    translationSpace=BCS_BoneSpace
    alphaBoolBlend {blendOption=Linear}
    boneToModify {boneName=up_uppelalm_l}
    translation {y=2, z=3}
   }
   n10 = "Transform (Modify) Bone - Bone: up_upperarm_r"(ComponentPose=@n9, Alpha=@n7, Translation="0,-7,-3", Rotation="0,0,0", Scale="1,1,1") {
    rotationMode=BMM_Additive
    rotationSpace=BCS_BoneSpace
    scaleMode=BMM_Additive
    scaleSpace=BCS_BoneSpace
    translationMode=BMM_Additive
    translationSpace=BCS_BoneSpace
    alphaBoolBlend {blendOption=Linear}
    boneToModify {boneName=up_upperarm_r}
    translation {y=-7, z=-3}
   }
   n11 = "Transform (Modify) Bone - Bone: back_uppelalm_l"(ComponentPose=@n10, Alpha=@n5, Translation="0,3,0", Rotation="0,0,0", Scale="1.5,1.5,1.5") {
    rotationMode=BMM_Additive
    rotationSpace=BCS_BoneSpace
    scaleMode=BMM_Additive
    scaleSpace=BCS_BoneSpace
    translationMode=BMM_Additive
    translationSpace=BCS_BoneSpace
    alphaBoolBlend {blendOption=Linear}
    boneToModify {boneName=back_uppelalm_l}
    scale=<1.5 1.5 1.5>
    translation {y=3}
   }
   n12 = "Transform (Modify) Bone - Bone: back_upperarm_r"(ComponentPose=@n11, Alpha=@n7, Translation="0,3,0", Rotation="0,0,0", Scale="1.5,1.5,1.5") {
    rotationMode=BMM_Additive
    rotationSpace=BCS_BoneSpace
    scaleMode=BMM_Additive
    scaleSpace=BCS_BoneSpace
    translationMode=BMM_Additive
    translationSpace=BCS_BoneSpace
    alphaBoolBlend {blendOption=Linear}
    boneToModify {boneName=back_upperarm_r}
    scale=<1.5 1.5 1.5>
    translation {y=3}
   }
   n13 = "Constraint - dyn_shoulderPlate_l_f (PP)"!(ComponentPose=@n12, Alpha=1, ConstraintWeights_0=0.5, ConstraintWeights_1:float=0.5) -> (Pose) {
    alphaBoolBlend {blendOption=Linear}
    boneToModify {boneName=dyn_shoulderPlate_l_f}
    constraintSetup [{targetBone {boneName=spine_05}}, {targetBone {boneName=clavicle_l}}]
    constraintWeights [1,1]
   }
   n14 = "Constraint - dyn_shourderPrate_r_f (PP)"!(ComponentPose=@n13, Alpha=1, ConstraintWeights_0=0.5, ConstraintWeights_1:float=0.5) -> (Pose) {
    alphaBoolBlend {blendOption=Linear}
    boneToModify {boneName=dyn_shourderPrate_r_f}
    constraintSetup [{targetBone {boneName=spine_05}}, {targetBone {boneName=clavicle_r}}]
    constraintWeights [1,1]
   }
   n15 = "Transform (Modify) Bone - Bone: dyn_shourderPrate_r_f"(ComponentPose=@n14, Alpha=Get Shoulder_rot_f_r(), Translation="-20,0,0", Rotation="20,-60,0", Scale="1,1,1") {
    rotationMode=BMM_Additive
    rotationSpace=BCS_BoneSpace
    scaleMode=BMM_Additive
    scaleSpace=BCS_BoneSpace
    translationMode=BMM_Additive
    translationSpace=BCS_BoneSpace
    alphaBoolBlend {blendOption=Linear}
    boneToModify {boneName=dyn_shourderPrate_r_f}
    rotation {pitch=20, yaw=-60}
    translation {x=-20}
   }
   n16 = "Transform (Modify) Bone - Bone: dyn_shourderPrate_r_f"(ComponentPose=@n15, Alpha=Get Shoulder_rot_f_r_0() -> (Shoulder_rot_f_r_0:double), Translation="5,0,0", Rotation="-30,40,0", Scale="1,1,1") {
    rotationMode=BMM_Additive
    rotationSpace=BCS_BoneSpace
    scaleMode=BMM_Additive
    scaleSpace=BCS_BoneSpace
    translationMode=BMM_Additive
    translationSpace=BCS_BoneSpace
    alphaBoolBlend {blendOption=Linear}
    boneToModify {boneName=dyn_shourderPrate_r_f}
    rotation {pitch=-30, yaw=40}
    translation {x=5}
   }
   n17 = "Transform (Modify) Bone - Bone: dyn_shoulderPlate_l_f"(ComponentPose=@n16, Alpha=Get Shoulder_rot_f_l(), Translation="-20,0,0", Rotation="20,60,0", Scale="1,1,1") {
    rotationMode=BMM_Additive
    rotationSpace=BCS_BoneSpace
    scaleMode=BMM_Additive
    scaleSpace=BCS_BoneSpace
    translationMode=BMM_Additive
    translationSpace=BCS_BoneSpace
    alphaBoolBlend {blendOption=Linear}
    boneToModify {boneName=dyn_shoulderPlate_l_f}
    rotation {pitch=20, yaw=60}
    translation {x=-20}
   }
   n18 = "Transform (Modify) Bone - Bone: dyn_shoulderPlate_l_f"(ComponentPose=@n17, Alpha=Get Shoulder_rot_f_l_0() -> (Shoulder_rot_f_l_0:double), Translation="7,0,0", Rotation="-30,-80,0", Scale="1,1,1") {
    rotationMode=BMM_Additive
    rotationSpace=BCS_BoneSpace
    scaleMode=BMM_Additive
    scaleSpace=BCS_BoneSpace
    translationMode=BMM_Additive
    translationSpace=BCS_BoneSpace
    alphaBoolBlend {blendOption=Linear}
    boneToModify {boneName=dyn_shoulderPlate_l_f}
    rotation {pitch=-30, yaw=-80}
    translation {x=7}
   }
   n19 = "Constraint - upperarmDn_l (PP)"!(ComponentPose=@n18, Alpha=1, ConstraintWeights_0=0.7, ConstraintWeights_1:float=0.3) -> (Pose) {
    alphaBoolBlend {blendOption=Linear}
    boneToModify {boneName=upperarmDn_l}
    constraintSetup [{targetBone {boneName=spine_04}}, {targetBone {boneName=clavicle_l}}]
    constraintWeights [1,1]
   }
   n20 = "Constraint - upperarmDn_r (PP)"!(ComponentPose=@n19, Alpha=1, ConstraintWeights_0=0.7, ConstraintWeights_1:float=0.3) -> (Pose) {
    alphaBoolBlend {blendOption=Linear}
    boneToModify {boneName=upperarmDn_r}
    constraintSetup [{targetBone {boneName=spine_04}}, {targetBone {boneName=clavicle_r}}]
    constraintWeights [1,1]
   }
   n21 = "Transform (Modify) Bone - Bone: upperarmDn_l"(ComponentPose=@n20, Alpha=Get Shoulder_rot_dn_l() -> (Shoulder_rot_dn_l:double), Translation="-8,0,0", Rotation="0,0,0", Scale="1,1,1") {
    rotationMode=BMM_Additive
    rotationSpace=BCS_BoneSpace
    scaleMode=BMM_Additive
    scaleSpace=BCS_BoneSpace
    translationMode=BMM_Additive
    translationSpace=BCS_BoneSpace
    alphaBoolBlend {blendOption=Linear}
    boneToModify {boneName=upperarmDn_l}
    translation {x=-8}
   }
   n22 = "Transform (Modify) Bone - Bone: upperarmDn_r"(ComponentPose=@n21, Alpha=Get Shoulder_rot_dn_r() -> (Shoulder_rot_dn_r:double), Translation="8,0,0", Rotation="0,0,0", Scale="1,1,1") {
    rotationMode=BMM_Additive
    rotationSpace=BCS_BoneSpace
    scaleMode=BMM_Additive
    scaleSpace=BCS_BoneSpace
    translationMode=BMM_Additive
    translationSpace=BCS_BoneSpace
    alphaBoolBlend {blendOption=Linear}
    boneToModify {boneName=upperarmDn_r}
    translation {x=8}
   }
   n23 = "Transform (Modify) Bone - Bone: upperarmDn_l"(ComponentPose=@n22, Alpha=Get Shoulder_rot_up_l() -> (Shoulder_rot_up_l:double), Translation="-3,0,15", Rotation="0,0,0", Scale="1,1,1") {
    rotationMode=BMM_Additive
    rotationSpace=BCS_BoneSpace
    scaleMode=BMM_Additive
    scaleSpace=BCS_BoneSpace
    translationMode=BMM_Additive
    translationSpace=BCS_BoneSpace
    alphaBoolBlend {blendOption=Linear}
    boneToModify {boneName=upperarmDn_l}
    translation {x=-3, z=15}
   }
   n24 = "Transform (Modify) Bone - Bone: upperarmDn_r"(ComponentPose=@n23, Alpha=Get Shoulder_rot_up_r() -> (Shoulder_rot_up_r:double), Translation="3,0,-15", Rotation="0,0,0", Scale="1,1,1") {
    rotationMode=BMM_Additive
    rotationSpace=BCS_BoneSpace
    scaleMode=BMM_Additive
    scaleSpace=BCS_BoneSpace
    translationMode=BMM_Additive
    translationSpace=BCS_BoneSpace
    alphaBoolBlend {blendOption=Linear}
    boneToModify {boneName=upperarmDn_r}
    translation {x=3, z=-15}
   }
   n25 = "Transform (Modify) Bone - Bone: upperarmDn_l"(
    ComponentPose=@n24
    Alpha=Get Shoulder_rot_back_l() -> (Shoulder_rot_back_l:double)
    Translation="-5,-5,-2.5"
    Rotation="0,0,0"
    Scale="1,1,1"
   ) {
    rotationMode=BMM_Additive
    rotationSpace=BCS_BoneSpace
    scaleMode=BMM_Additive
    scaleSpace=BCS_BoneSpace
    translationMode=BMM_Additive
    translationSpace=BCS_BoneSpace
    alphaBoolBlend {blendOption=Linear}
    boneToModify {boneName=upperarmDn_l}
    translation=<-5 -5 -2.5>
   }
   n26 = "Transform (Modify) Bone - Bone: upperarmDn_r"(ComponentPose=@n25, Alpha=Get Shoulder_rot_back_r() -> (Shoulder_rot_back_r:double), Translation="5,5,2.5", Rotation="0,0,0", Scale="1,1,1") {
    rotationMode=BMM_Additive
    rotationSpace=BCS_BoneSpace
    scaleMode=BMM_Additive
    scaleSpace=BCS_BoneSpace
    translationMode=BMM_Additive
    translationSpace=BCS_BoneSpace
    alphaBoolBlend {blendOption=Linear}
    boneToModify {boneName=upperarmDn_r}
    translation=<5 5 2.5>
   }
   Component To Local(ComponentPose=@n26)
   Control Rig() {
    alphaBoolBlend {blendOption=Linear}
    controlRigAssetReference {blueprintRigClass=/Game/Characters/Player/Female/Medium/Bodies/F_MED_LastVoiceSteel/Meshes/F_MED_LastVoiceSteel_CtrlRig.F_MED_LastVoiceSteel_CtrlRig_C}
   }
   n27 = Layered blend per bone(
    BasePose=Use cached pose 'SavedPose'()
    BlendPoses_0=Slot 'DefaultSlot'()
    BlendWeights_0=Get PlayingFrontendIdle() -> (PlayingFrontendIdle:double)
   ) {layerSetup [{branchFilters [{boneName=dyn_jacket_r_B_Master}, {boneName=dyn_jacket_r_A_Master}]}]}
   Output Pose(Result=@n27) {layerGroup=None}
   n28 = Control Rig(Source=Copy Pose From Mesh() {bUseAttachedParent=true}) {
    alphaBoolBlend {blendOption=Linear}
    controlRigAssetReference {blueprintRigClass=&/F_MED_LastVoiceSteel_DeformRig_ControlRig.F_MED_LastVoiceSteel_DeformRig_ControlRig_C}
   }
   n29 = Control Rig(Source=@n28) {
    alphaBoolBlend {blendOption=Linear}
    controlRigAssetReference {blueprintRigClass=&/F_MED_LastVoiceSteel_Jacket_CtrlRig.F_MED_LastVoiceSteel_Jacket_CtrlRig_C}
   }
   n30 = Get GravityOverride()
   n31 = RigidBody!(ComponentPose=Local To Component(LocalPose=@n29), Alpha, OverrideWorldGravity:Vector=@n30, ExternalForce) -> (Pose) {
    bClampLinearTranslationLimitToRefPose=true
    bOverrideWorldGravity=true
    overridePhysicsAsset=&/F_MED_LastVoiceSteel_Physics
    simulationSpace=BaseBoneSpace
    alphaBoolBlend {blendOption=Linear}
    baseBoneRef {boneName=spine_01}
    componentAppliedLinearAccClamp=<5000 5000 5000>
    componentLinearAccScale {z=0.6}
    componentLinearVelScale {z=0.6}
    overrideWorldGravity {z=-980}
    simSpaceSettings {
     maxAngularAcceleration=500
     maxAngularVelocity=3
     maxLinearAcceleration=700
     maxLinearVelocity=500
     velocityScaleZ=0
     worldAlpha=0.25
     externalLinearDragV=<1 1>
    }
   }
   n32 = RigidBody(ComponentPose=@n31, OverrideWorldGravity=@n30) {
    bClampLinearTranslationLimitToRefPose=true
    bOverrideWorldGravity=true
    overridePhysicsAsset=&/F_MED_LastVoiceSteel_JacketFlaps_Physics
    simulationSpace=BaseBoneSpace
    alphaBoolBlend {blendOption=Linear}
    baseBoneRef {boneName=spine_05}
    componentAppliedLinearAccClamp=<5000 5000 5000>
    componentLinearAccScale {z=0.6}
    componentLinearVelScale {z=0.6}
    overrideWorldGravity {z=-980}
    simSpaceSettings {
     maxAngularAcceleration=500
     maxAngularVelocity=3
     maxLinearAcceleration=700
     maxLinearVelocity=500
     velocityScaleZ=0
     worldAlpha=0.25
     externalLinearDragV=<1 1>
    }
   }
   Save cached pose 'SavedPose'(Pose=Component To Local(ComponentPose=@n32)) {cacheName=SavedPose}
  ]}
 EventGraph {nodes [
   n1 = Set WPO Globals Function(
    Material=Get Mat Instance Dynamic()
    Overlay Material=Get Overlay Mat Instance Dynamic()
    Skeletal Mesh=Get SkelMesh()
    "Local Noise Transform  Location"=Get Local Noise Transform Location() -> (Local Noise Transform Location:Vector)
    Local Noise Offset=Get Local Noise Offset()
   )
   n2 = Event BlueprintUpdateAnimation()
   Set WPO Regional Function!(
    execute=@n1.then
    Material:Material Instance Dynamic Object Reference=Get Mat Instance Dynamic()
    Overlay Material:Material Instance Dynamic Object Reference=Get Overlay Mat Instance Dynamic()
    Skeletal Mesh:Skeletal Mesh Component Object Reference=Get SkelMesh()
    Velocity Transform:Name=spine_05
    Velocity Interp Speed:double=Get Velocity Interp Speed()
    Collision Transform:Name=root
    Collision Transform Axis:EAxis Enum
    Invert Collision Transform Axis:bool=false
    Delta Time X:double=@n2.DeltaTimeX
    Previous Velocity Transform Location:"Vector (by ref)"=Get R1 Velocity Transform Location()
    Previous Velocity Transform Velocity:"Vector (by ref)"=Get R1 Velocity Transform Velocity()
    Region:int=1
   ) -> (then)
   > n3 = Branch(Condition=Get bIsSkydiveFloating())
   > n4 = Set WPO Pho Wind Intensity(WPO Pho Wind Intensity=1)
   SetScalarParameterValue(execute=[@n5.then, @n4.then], self=Get Mat Instance Dynamic(), ParameterName=ClothSim_WindStrength, Value=Get WPO Pho Wind Intensity())
   n5 = Set WPO Pho Wind Intensity(execute=@n3.else, WPO Pho Wind Intensity=0)
   Event BlueprintInitializeAnimation()
   > n6 = Gravity Override State Function DEPRECATED(
    CharacterPartAnimInstance=Self-Reference()
    GameTimeInSeconds=GetGameTimeInSeconds()
    GravityOverrideParams=Get GravityOverrideParams() -> (GravityOverrideParams:Gravity Override Params Struct Structure)
   )
   > Set GravityOverride(GravityOverride=@n6)
   Set Skel_Mesh(Skel_Mesh=GetOwnerSkeletalMeshComponent())
   > n7 = Cast To PlayerPawn_Athena(Object=TryGetPawnOwner())
   > n8 = Set Pawn Owner(Pawn Owner:Fort Player Pawn Athena Object Reference=@n7) -> (Output_Get:Fort Player Pawn Athena Object Reference)
   > Set Shoulder_Mesh(Shoulder_Mesh:Skeletal Mesh Component Object Reference=GetSkeletalMeshForPartType(self=@n8)) -> (Output_Get:Skeletal Mesh Component Object Reference)
   n9 = Set SkelMesh(SkelMesh=GetPartSkeletalMeshComponent())
   > n10 = Get Dynamic Material Instance(SkeletalMeshComponent=@n9)
   > Set Mat Instance Dynamic(Mat Instance Dynamic=@n10)
   > n11 = GetOrCreateDynamicOverlayMaterial(MeshComponent=@n9, MaterialName=None)
   > Set Overlay Mat Instance Dynamic(Overlay Mat Instance Dynamic=@n11)
   n12 = Get Shoulder_Mesh() -> (Shoulder_Mesh:Skeletal Mesh Component Object Reference)
   Set Shoulder_rot_f_r(Shoulder_rot_f_r:double=GetAlphaFromReferenceTransformDelta(SkeletalMeshComponent=@n12, BoneName=upperarm_r, BaseName=clavicle_r, InputRange=(X=0,Y=110), TwistAxis="0,0,1")) -> (Output_Get:double)
   > Set Shoulder_rot_f_l(Shoulder_rot_f_l:double=GetAlphaFromReferenceTransformDelta(SkeletalMeshComponent=@n12, BoneName=upperarm_l, BaseName=clavicle_l, InputRange=(X=0,Y=110), TwistAxis="0,0,1")) -> (Output_Get:double)
   n13 = Get Shoulder_Mesh()
   > Set Shoulder_rot_up_r(Shoulder_rot_up_r:double=GetAlphaFromReferenceTransformDelta(SkeletalMeshComponent=@n13, BoneName=upperarm_r, BaseName=clavicle_r, DeltaComponentType=EulerY, InputRange=(X=0,Y=130), TwistAxis="0,0,1")) -> (Output_Get:double)
   > Set Shoulder_rot_up_l(Shoulder_rot_up_l:double=GetAlphaFromReferenceTransformDelta(SkeletalMeshComponent=@n13, BoneName=upperarm_l, BaseName=clavicle_l, DeltaComponentType=EulerY, InputRange=(X=0,Y=130), TwistAxis="0,0,1")) -> (Output_Get:double)
   n14 = Get Shoulder_Mesh()
   > Set Shoulder_rot_dn_r(Shoulder_rot_dn_r:double=GetAlphaFromReferenceTransformDelta(SkeletalMeshComponent=@n14, BoneName=upperarm_r, BaseName=clavicle_r, DeltaComponentType=EulerY, InputRange=(X=10,Y=-90), TwistAxis="0,0,1")) -> (Output_Get:double)
   > Set Shoulder_rot_dn_l(Shoulder_rot_dn_l:double=GetAlphaFromReferenceTransformDelta(SkeletalMeshComponent=@n14, BoneName=upperarm_l, BaseName=clavicle_l, DeltaComponentType=EulerY, InputRange=(X=10,Y=-90), TwistAxis="0,0,1")) -> (Output_Get:double)
   n15 = Get Shoulder_Mesh()
   > Set Shoulder_rot_back_r(Shoulder_rot_back_r:double=GetAlphaFromReferenceTransformDelta(SkeletalMeshComponent=@n15, BoneName=upperarm_twist_01_r, BaseName=clavicle_r, InputRange=(X=10,Y=-30), TwistAxis="0,0,1")) -> (Output_Get:double)
   > Set Shoulder_rot_back_l(Shoulder_rot_back_l:double=GetAlphaFromReferenceTransformDelta(SkeletalMeshComponent=@n15, BoneName=upperarm_twist_01_l, BaseName=clavicle_l, InputRange=(X=10,Y=-30), TwistAxis="0,0,1")) -> (Output_Get:double)
   n16 = Get Shoulder_Mesh()
   > Set Shoulder_rot_f_r_0(Shoulder_rot_f_r_0:double=GetAlphaFromReferenceTransformDelta(SkeletalMeshComponent=@n16, BoneName=upperarm_r, BaseName=clavicle_r, DeltaComponentType=EulerX, InputRange=(X=0,Y=-140), TwistAxis="0,0,1")) -> (Output_Get:double)
   > Set Shoulder_rot_f_l_0(Shoulder_rot_f_l_0:double=GetAlphaFromReferenceTransformDelta(SkeletalMeshComponent=@n16, BoneName=upperarm_l, BaseName=clavicle_l, DeltaComponentType=EulerX, InputRange=(X=0,Y=-140), TwistAxis="0,0,1")) -> (Output_Get:double)
   Event OnCharacterCustomizationLoaded()
   Reroute Node(InputPin:float=@n2.DeltaTimeX) -> (OutputPin:float)
   Set PlayingFrontendIdle(execute=@n17.then, PlayingFrontendIdle:double=GetCurveValue(self=GetOwnerFrontEndAnimInstance(), CurveName=FrontEnd_Idle_Bypass)) -> (Output_Get:double)
   n17 = Branch(execute=@n2.then, Condition=Get bIsInFrontEnd())
  ]}
]
variables [
 GravityOverride {type=Vector}
 GravityOverrideParams {type=GravityOverrideParamsStruct}
 Local Noise Offset {type=Vector}
 Local Noise Transform Location {type=Vector}
 Mat Instance Dynamic {type=MaterialInstanceDynamic}
 Overlay Mat Instance Dynamic {type=MaterialInstanceDynamic}
 Pawn Owner {type=FortPlayerPawnAthena}
 PlayingFrontendIdle {type=number}
 R1 Velocity Transform Location {type=Vector}
 R1 Velocity Transform Velocity {type=Vector}
 Shoulder_Mesh {type=SkeletalMeshComponent}
 Shoulder_rot_back_l {type=number}
 Shoulder_rot_back_r {type=number}
 Shoulder_rot_dn_l {type=number}
 Shoulder_rot_dn_r {type=number}
 Shoulder_rot_f_l {type=number}
 Shoulder_rot_f_l_0 {type=number}
 Shoulder_rot_f_r {type=number}
 Shoulder_rot_f_r_0 {type=number}
 Shoulder_rot_up_l {type=number}
 Shoulder_rot_up_r {type=number}
 Skel_Mesh {type=SkeletalMeshComponent}
 SkelMesh {type=SkeletalMeshComponent}
 Velocity Interp Speed {type=number, default=10}
 WPO Pho Wind Intensity {type=number}
]
