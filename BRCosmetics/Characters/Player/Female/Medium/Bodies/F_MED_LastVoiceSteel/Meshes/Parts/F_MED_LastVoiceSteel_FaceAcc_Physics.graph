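# Physics asset for the face accessory mesh of the female medium LastVoiceSteel body, covering two dynamic hair strands.
# Left and right chains each link a small master body to the head, followed by three capsule segments in series.
# Hair joints have swing limits, tighter on one axis, and zero twist.
# Larger head and neck_01 capsules form the body volumes, joined by an angularly free constraint.
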PhysicsAsset /BRCosmetics/Characters/Player/Female/Medium/Bodies/F_MED_LastVoiceSteel/Meshes/Parts/F_MED_LastVoiceSteel_FaceAcc_Physics
bodies {
 dyn_hair_l_A_01 [{shapeType=Capsule, length=8, radius=2, shapeName=dyn_hair_l_A_01_capsule, center=<0 0.1464 0.83>, rotation=<rot 90 -90 0>}]
 dyn_hair_l_A_02 [{shapeType=Capsule, length=5, radius=2, shapeName=dyn_hair_l_A_01_capsule, center=<0 0.1912 1.0845>, rotation=<rot 90 90.0001 0>}]
 dyn_hair_l_A_03 [{shapeType=Capsule, length=4, radius=2, shapeName=dyn_hair_l_A_01_capsule, center=<0 0.2651 1.5032>, rotation=<rot 90 90.0001 0>}]
 dyn_hair_l_A_Master [{shapeType=Capsule, length=0.505, radius=0.505, shapeName=dyn_hair_l_A_Master_capsule, rotation=<rot 0 -90 0>}]
 dyn_hair_r_A_01 [{shapeType=Capsule, length=6, radius=2, shapeName=dyn_hair_l_A_01_capsule, center=<4.0791 0 -1.2256>, rotation=<rot 0 0 90>}]
 dyn_hair_r_A_02 [{shapeType=Capsule, length=5, radius=2, shapeName=dyn_hair_l_A_01_capsule, center=<2.7976 0 -1.4>, rotation=<rot 0 0 90>}]
 dyn_hair_r_A_03 [{shapeType=Capsule, length=4, radius=2, shapeName=dyn_hair_l_A_01_capsule, center=<2.119 0 -1.5>, rotation=<rot 0 0 90>}]
 dyn_hair_r_A_Master [{shapeType=Capsule, length=0.505, radius=0.505, shapeName=dyn_hair_r_A_Master_capsule, rotation=<rot 0 -90 0>}]
 head [{shapeType=Capsule, length=4, radius=6, shapeName=head_capsule, center=<2.8313 0 0>, rotation=<rot 0 90 90>}]
 neck_01 [{shapeType=Capsule, length=15, radius=8, shapeName=spine_05_capsule, center=<1.3215 -7.9621 0>, rotation=<rot 0 -90 0>}]
}
constraints [
 {bone1Name=dyn_hair_l_A_Master, bone2Name=head, swing1LimitDegrees=45, swing2LimitDegrees=5, twistLimitDegrees=0}
 {bone1Name=dyn_hair_l_A_01, bone2Name=dyn_hair_l_A_Master, swing1LimitDegrees=45, swing2LimitDegrees=15, twistLimitDegrees=0}
 {bone1Name=dyn_hair_l_A_02, bone2Name=dyn_hair_l_A_01, swing1LimitDegrees=45, swing2LimitDegrees=15, twistLimitDegrees=0}
 {bone1Name=dyn_hair_l_A_03, bone2Name=dyn_hair_l_A_02, swing1LimitDegrees=45, swing2LimitDegrees=15, twistLimitDegrees=0}
 {bone1Name=dyn_hair_r_A_Master, bone2Name=head, swing1LimitDegrees=45, swing2LimitDegrees=5, twistLimitDegrees=0}
 {bone1Name=dyn_hair_r_A_01, bone2Name=dyn_hair_r_A_Master, swing1LimitDegrees=45, swing2LimitDegrees=15, twistLimitDegrees=0}
 {bone1Name=dyn_hair_r_A_02, bone2Name=dyn_hair_r_A_01, swing1LimitDegrees=45, swing2LimitDegrees=15, twistLimitDegrees=0}
 {bone1Name=dyn_hair_r_A_03, bone2Name=dyn_hair_r_A_02, swing1LimitDegrees=45, swing2LimitDegrees=15, twistLimitDegrees=0}
 {
  bone1Name=head
  bone2Name=neck_01
  swing1LimitDegrees=45
  swing1Motion=Free
  swing2LimitDegrees=45
  swing2Motion=Free
  twistLimitDegrees=45
  twistMotion=Free
 }
]
